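# Translucent unlit flame trail material for the CanvasPrint character's FX, usable on Niagara ribbons and mesh particles.
# Two animated sine-band glitch masks offset the UVs of a panned trail texture and add tinted color to particle color.
# Opacity combines that texture with caustics-noise dissolve, tip and length fades, and depth fade on higher-end paths.
# World position offset mixes SplineThicken, sine wobble and an optional pull, with Niagara dynamic parameters driving much of the effect.
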Material /BRCosmetics/Characters/Player/Male/Medium/Bodies/M_MED_CanvasPrint/FX/M_CanvasPrint_Flame_Trail
graph {
 properties {
  bUsedWithNiagaraMeshParticles=true
  bUsedWithNiagaraRibbons=true
  blendMode=BLEND_Translucent
  pixelDepthOffsetMode=PDOM_Legacy
  shadingModel=MSM_Unlit
  twoSided=true
  usedShadingModels=Unlit
 }
 nodes [
  n1 = Constant()
  n2 = DynamicParameter() {defaultValue=<rgba 1 1 1 0>, paramNames [WPO, Speed, WPO Disturbance, U offset]}
  n3 = Add(A=TextureCoordinate(), B=@n2.U offset)
  n4 = Time()
  n5 = Multiply(A=@n4) {constB=-0.3}
  n6 = AppendVector(A=@n5, B=Multiply(A=@n5) {constB=0.4})
  n7 = Sine(None=ComponentMask(None=Add(A=@n3, B=@n6)) {g=true}) {period=1.6}
  n8 = Round(None=Power(Base=Sine(None=@n7) {period=1.2}) {constExponent=1.2})
  n9 = Multiply(A=@n4) {constB=-1.8}
  n10 = Multiply(A=@n9) {constB=0.4}
  n11 = ComponentMask(None=Add(A=@n3, B=AppendVector(A=@n9, B=@n10))) {r=true}
  n12 = Power(Base=Sine(None=Sine(None=@n11) {period=1.3}) {period=2}) {constExponent=1}
  n13 = DynamicParameter() {parameterIndex=2, defaultValue=<rgba 1 2 1 1>, paramNames [Glitch, FadeDistance, Opacity_PWR, GlitchColorsMult]}
  n14 = Multiply(A=Multiply(A=Multiply(A=@n8), B=Multiply(A=@n12)), B=@n13.Glitch)
  n15 = Multiply(A=@n14) {constB=2.5}
  n16 = ScalarParameter() {defaultValue=10, parameterName=WidthBase}
  n17 = DynamicParameter() {defaultValue=<rgba 1 1 1 0>, paramNames [WPO, Speed, WPO Disturbance, U offset]}
  n18 = ScalarParameter() {defaultValue=40, parameterName=WidthTip}
  n19 = MaterialFunctionCall(WidthBase=Multiply(A=Multiply(A=@n16, B=@n17.WPO)), WidthTip=Multiply(A=Multiply(A=@n18, B=@n17.WPO)), Expand U or V UV Channel=StaticBool()) {materialFunction=&SplineThicken}
  n20 = DynamicParameter() {defaultValue=<rgba 1 1 1 0>, paramNames [WPO, Speed, WPO Disturbance, U offset]}
  n21 = Constant()
  n22 = ComponentMask(None=Add(A=TextureCoordinate() {uTiling=2.2}, B=AppendVector(A=@n20.U offset, B=@n21))) {r=true}
  n23 = Multiply(A=Time()) {constB=-1.8}
  n24 = Multiply(A=Sine(None=Add(A=@n22, B=@n23)) {period=1.2}, B=VectorParameter() {parameterName=WPO Direction, defaultValue=<rgba 0 0 1 1>}.RGB)
  n25 = Transform(None=@n24) {transformSourceType=TRANSFORMSOURCE_World, transformType=TRANSFORM_Camera}
  n26 = Transform(None=@n24) {transformSourceType=TRANSFORMSOURCE_Local}
  n27 = ScalarParameter() {parameterName=Change Transform}
  n28 = Multiply(A=ComponentMask(None=TextureCoordinate()) {r=true}) {constB=3}
  n29 = Multiply(
   A=Multiply(
    A=LinearInterpolate(A=@n25, B=@n26, Alpha=@n27)
    B=DynamicParameter() {defaultValue=<rgba 1 1 1 0>, paramNames [WPO, Speed, WPO Disturbance, U offset]}.WPO Disturbance
   )
   B=LinearInterpolate(Alpha=Saturate(None=@n28))
  )
  n30 = DynamicParameter() {parameterIndex=1, defaultValue=<rgba 1 1 1 0>, paramNames [X,Y,Z,WPOIntensity]}
  n31 = MaterialFunctionCall(X=@n30.X, Y=@n30.X, Z=@n30.X) {materialFunction=&MakeFloat3}
  n32 = ParticlePositionWS()
  n33 = ComponentMask(None=TextureCoordinate()) {r=true}
  n34 = ScalarParameter() {defaultValue=2, parameterName=WPOBlendPower}
  n35 = LinearInterpolate(
   B=Clamp(None=Subtract(A=@n31.Result, B=@n32), Min=Constant() {r=-30}, Max=Constant() {r=30})
   Alpha=Multiply(A=@n30.WPOIntensity, B=Power(Base=@n33, Exp=@n34))
  )
  n36 = ScalarParameter() {parameterName=Use Interpolate WPO}
  Add(
   A=Multiply(A=MaterialFunctionCall(R=@n1, G=@n1, B=@n15) {materialFunction=&AppendMany}.RGB)
   B=Add(A=Add(A=@n19.WorldPositionOffset, B=@n29) {constB=0}, B=LinearInterpolate(B=@n35, Alpha=@n36))
  )
  Constant()
  Constant()
  n37 = Add(A=TextureCoordinate(), B=@n2.U offset)
  n38 = ComponentMask(None=@n37) {r=true}
  n39 = Saturate(None=Power(Base=OneMinus(None=@n38)) {constExponent=1.25})
  n40 = VectorParameter() {parameterName=Glitch_Color_02, defaultValue=<rgba 0.7501 0 1 1>}
  n41 = DynamicParameter() {parameterIndex=3, defaultValue=<rgba 0 1 1 1>, paramNames [VariantGlitchColor, Param2, Param3, Param4]}
  n42 = DynamicParameter() {parameterIndex=2, defaultValue=<rgba 1 2 1 1>, paramNames [Glitch, FadeDistance, Opacity_PWR, GlitchColorsMult]}
  n43 = Multiply(
   A=Multiply(A=Saturate(None=@n14), B=@n39)
   B=Multiply(A=LinearInterpolate(A=@n40.RGB, Alpha=@n41.VariantGlitchColor), B=@n42.GlitchColorsMult)
  )
  n44 = Time()
  n45 = Multiply(A=@n44) {constB=-0.2}
  n46 = AppendVector(A=@n45, B=Multiply(A=@n45) {constB=0.4})
  n47 = Sine(None=ComponentMask(None=Add(A=@n37, B=@n46)) {g=true}) {period=1.8}
  n48 = Round(None=Power(Base=Sine(None=@n47) {period=1.6}) {constExponent=1})
  n49 = Multiply(A=@n44) {constB=-1.6}
  n50 = Multiply(A=@n49) {constB=0.4}
  n51 = ComponentMask(None=Add(A=@n37, B=AppendVector(A=@n49, B=@n50))) {r=true}
  n52 = Power(Base=Sine(None=Sine(None=@n51)) {period=2}) {constExponent=1.2}
  n53 = Multiply(A=Multiply(A=Multiply(A=@n48), B=Multiply(A=@n52)), B=@n42.Glitch)
  n54 = VectorParameter() {parameterName=Glitch_Color_01, defaultValue=<rgba 0 0.6272 0.6458 1>}
  n55 = DynamicParameter() {parameterIndex=3, defaultValue=<rgba 0 1 1 1>, paramNames [VariantGlitchColor, Param2, Param3, Param4]}
  n56 = Multiply(
   A=Multiply(A=@n39, B=Saturate(None=@n53))
   B=Multiply(A=@n42.GlitchColorsMult, B=LinearInterpolate(A=@n54.RGB, Alpha=@n55.VariantGlitchColor))
  )
  n57 = Power(Base=ComponentMask(None=TextureCoordinate()) {r=true}) {constExponent=0.8}
  n58 = ParticleColor()
  n59 = Add(A=Multiply(A=Add(A=@n43, B=@n56), B=Saturate(None=@n57)), B=@n58.RGB)
  MaterialFunctionCall(Emissive=@n59) {materialFunction=&MF_FX_EmissiveAdjustments}
  n60 = Constant() {r=1}
  Multiply(
   A=MaterialFunctionCall(
    Color_Morning=@n60
    Color_Day=@n60
    Color_Evening=@n60
    Color_Night=VectorParameter() {parameterName="Color_Night (V3)", defaultValue=<rgba 1.5 1.5 1.5 1>}.RGB
   ) {materialFunction=&Color_Time_of_Day}.Result
   B=@n59
  )
  n61 = Constant()
  n62 = AppendVector(A=Multiply(A=Multiply(A=@n14) {constB=0.13}, B=@n39), B=@n61)
  n63 = AppendVector(A=Multiply(A=Multiply(A=@n53) {constB=0.06}, B=@n39), B=@n61)
  n64 = Multiply(
   A=Add(A=Add(A=@n62, B=@n63), B=TextureCoordinate())
   B=ComponentMask(None=VectorParameter() {parameterName=UV Scale, defaultValue=<rgba 0.8 1 0 0>}.RGB) {g=true, r=true}
  )
  n65 = AppendVector(A=Power(Base=ComponentMask(None=@n64) {r=true}) {constExponent=0.7}, B=ComponentMask(None=@n64) {g=true})
  n66 = AppendVector(A=DynamicParameter() {defaultValue=<rgba 1 1 1 0>, paramNames [WPO, Speed, WPO Disturbance, U offset]}.U offset, B=Constant())
  n67 = Constant2Vector() {r=-0.3}
  n68 = DynamicParameter() {defaultValue=<rgba 1 1 1 0>, paramNames [WPO, Speed, WPO Disturbance, U offset]}
  n69 = TextureSampleParameter2D(UVs=Panner(Coordinate=Add(A=@n65, B=@n66), Speed=Multiply(A=@n67, B=@n68.Speed)) {speedX=-0.3}) {
   parameterName=Tex_Default
   samplerType=SAMPLERTYPE_LinearGrayscale
   texture=/BRCosmetics/Athena/Effects/Particles/Skydiving/NS_Contrail_Nebula/T_Nebula_Trail
  }
  n70 = Power(
   Base=ChannelMaskParameter(None=@n69.RGBA) {group=Tex, parameterName=Mask}.R
   Exp=ScalarParameter() {defaultValue=1, group=Tex, parameterName=Tex alpha fix}
  ) {constExponent=2.2}
  n71 = ComponentMask(None=TextureCoordinate()) {r=true}
  n72 = Saturate(None=Multiply(A=OneMinus(None=@n71)) {constB=5})
  n73 = DynamicParameter() {defaultValue=<rgba 1 1 1 0>, paramNames [WPO, Speed, WPO Disturbance, U offset]}
  n74 = Constant()
  n75 = Panner(Coordinate=Add(A=TextureCoordinate() {uTiling=0.7, vTiling=0.8}, B=AppendVector(A=@n73.U offset, B=@n74)), Time=@n73.Speed) {speedX=-0.5}
  n76 = ComponentMask(None=TextureCoordinate()) {r=true}
  n77 = LinearInterpolate(
   A=ScalarParameter() {parameterName=Dissolve mask Min}
   B=ScalarParameter() {defaultValue=0.5, parameterName=Dissolve mask Max}
   Alpha=Power(Base=@n76) {constExponent=0.9}
  ) {constA=0.1, constB=2}
  n78 = LinearInterpolate(
   A=ScalarParameter() {defaultValue=-2, parameterName=Disappear min}
   B=ScalarParameter() {defaultValue=2, parameterName=Disappear max}
   Alpha=Power(
    Base=TextureSample(UVs=@n75) {samplerType=SAMPLERTYPE_LinearGrayscale, texture=/Game/Global/Textures/Default/Caustics/T_Nosie_Caustics_Grayscale}.R
    Exp=Saturate(None=@n77)
   )
  ) {constA=-2, constB=2}
  n79 = OneMinus(None=@n76)
  n80 = Multiply(A=Multiply(A=Multiply(A=@n70, B=@n72), B=@n58.A), B=Saturate(None=Multiply(A=@n78, B=@n79)))
  n81 = TextureCoordinate()
  n82 = Constant2Vector() {g=0.5, r=1}
  n83 = LinearInterpolate(
   B=OneMinus(None=SphereMask(A=@n81, B=@n82) {attenuationRadius=0.6, hardnessPercent=50})
   Alpha=ScalarParameter() {defaultValue=1, parameterName=Fixed tips}
  ) {constA=1}
  n84 = Multiply(A=ComponentMask(None=TextureCoordinate()) {r=true}) {constB=15}
  n85 = DynamicParameter() {parameterIndex=2, defaultValue=<rgba 1 2 1 1>, paramNames [Glitch, FadeDistance, Opacity_PWR, GlitchColorsMult]}
  n86 = Power(Base=Multiply(A=Multiply(A=@n80, B=@n83), B=Saturate(None=@n84)), Exp=@n85.Opacity_PWR)
  n87 = DepthFade(Opacity=@n86, FadeDistance=@n85.FadeDistance)
  n88 = QualitySwitch(Default=@n87, Low=@n86, High=@n87, Medium=@n87, Epic=@n87)
  n89 = OneMinus(None=ComponentMask(None=TextureCoordinate()) {r=true})
  n90 = ScalarParameter() {defaultValue=0.5, parameterName=OpacityGradient_PWR}
  Multiply(A=ShadingPathSwitch(Default=@n88, Deferred=@n88, Forward=@n86, Mobile=@n86), B=Saturate(None=Power(Base=@n89, Exp=@n90) {constExponent=1}))
  Round(None=@n52)
  Round(None=@n12)
  n91 = ComponentMask(None=TextureCoordinate()) {r=true}
  n92 = ScalarParameter() {parameterName=SplineThicken Edge Strenght}
  Saturate(None=Multiply(A=LinearInterpolate(A=@n91, Alpha=@n92), B=ScalarParameter() {defaultValue=3, parameterName=SplineTicken Edge Contrast}) {constB=5})
  ScalarParameter() {defaultValue=2, parameterName=FadeDistance}
 ]
}
