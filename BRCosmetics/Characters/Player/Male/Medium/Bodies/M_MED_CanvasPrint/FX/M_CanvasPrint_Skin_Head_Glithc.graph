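# Translucent unlit material for Niagara mesh particles that draws a two-color glitch effect for the CanvasPrint character's head.
# Time-driven sines build banded horizontal UV distortion and a pulsing X scale, both weighted by the Glitch dynamic parameter.
# Subtracting two offset samples of the glitch texture in both directions yields edge masks tinted by Color_01 and Color_02.
# Glitch also modulates brightness and opacity, with particle color multiplied into both.
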Material /BRCosmetics/Characters/Player/Male/Medium/Bodies/M_MED_CanvasPrint/FX/M_CanvasPrint_Skin_Head_Glithc
graph {
 properties {
  bAutomaticallySetUsageInEditor=false
  bUsedWithNiagaraMeshParticles=true
  blendMode=BLEND_Translucent
  shadingModel=MSM_Unlit
  usedShadingModels=Unlit
 }
 nodes [
  n1 = Time()
  n2 = Sine(None=Sine(None=Multiply(A=@n1) {constB=2}) {period=1.2}) {period=1.6}
  n3 = Add(A=TextureCoordinate(), B=@n2)
  n4 = Power(Base=Sine(None=ComponentMask(None=@n3) {g=true}) {period=0.4})
  n5 = Constant()
  n6 = AppendVector(A=Multiply(A=Saturate(None=@n4)) {constB=0.025}, B=@n5)
  Add(A=@n6, B=AppendVector(A=Constant() {r=-0.05}, B=@n5))
  n7 = Constant() {r=-0.022}
  n8 = Constant()
  n9 = DynamicParameter() {paramNames [Glitch,Param2,Param3,Param4]}
  n10 = Add(A=Add(A=TextureCoordinate(), B=AppendVector(A=@n7, B=@n8)), B=Multiply(A=@n6, B=@n9.Glitch))
  n11 = Constant() {r=0.01}
  n12 = Multiply(A=@n11, B=ScalarParameter() {defaultValue=5.5, parameterName=Offset})
  n13 = Multiply(A=@n2, B=DynamicParameter() {paramNames [Glitch,Param2,Param3,Param4]}.Glitch)
  n14 = LinearInterpolate(Alpha=@n13) {constA=1, constB=1.06}
  n15 = Constant() {r=1}
  n16 = TextureSample(UVs=MaterialFunctionCall(UVs=Add(A=@n10, B=@n12) {constB=-0.01}, Texture Scale=AppendVector(A=@n14, B=@n15)) {materialFunction=&ScaleUVsByCenter}.UVs) {
   samplerSource=SSM_Clamp_WorldGroupSettings
   texture=&/T_CanvasPrint_Head_Glitches
  }
  n17 = TextureSample(UVs=@n10) {
   samplerSource=SSM_Clamp_WorldGroupSettings
   texture=&T_CanvasPrint_Head_Glitches
  }
  n18 = Saturate(None=Subtract(A=@n16.R, B=@n17.R))
  n19 = Saturate(None=Subtract(A=@n17.R, B=@n16.R))
  n20 = Add(
   A=Multiply(A=VectorParameter() {parameterName=Color_02, defaultValue=<rgba 4.5009 0 6 1>}.RGB, B=@n18)
   B=Multiply(A=VectorParameter() {parameterName=Color_01, defaultValue=<rgba 0 4.3698 4.5 1>}.RGB, B=@n19)
  )
  n21 = Multiply(A=@n20) {constB=0.55}
  n22 = Multiply(A=@n20) {constB=2}
  n23 = DynamicParameter() {paramNames [Glitch,Param2,Param3,Param4]}
  n24 = ParticleColor()
  MaterialFunctionCall(Emissive=Multiply(A=LinearInterpolate(A=@n21, B=@n22, Alpha=@n23.Glitch), B=@n24.RGB)) {materialFunction=&MF_FX_EmissiveAdjustments}
  n25 = Add(A=@n18, B=@n19)
  n26 = Saturate(None=@n13)
  Multiply(A=Multiply(A=Saturate(None=@n25), B=LinearInterpolate(Alpha=@n26) {constA=1.4, constB=0.2}), B=@n24.A) {constB=0.4}
  n27 = Multiply(A=@n1) {constB=3}
  n28 = Constant() {r=0.3}
  n29 = Constant() {r=1}
  Multiply(
   A=@n11
   B=LinearInterpolate(
    B=MaterialFunctionCall!(Sine Phase=@n27, Value 1=@n28, Value 2=@n29) {materialFunction=/Engine/Functions/Engine_MaterialFunctions02/Utility/Sine_Remapped}.Remapped Sine
    Alpha=@n9.Glitch
   ) {constA=1}
  )
 ]
}
