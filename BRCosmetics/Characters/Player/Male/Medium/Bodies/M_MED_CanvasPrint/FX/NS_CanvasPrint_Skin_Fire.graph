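# Niagara system that attaches purple fire meshes to the head and body of the CanvasPrint player cosmetic.
# Face and FireBody emitters burst persistent mesh particles that follow skeletal sockets or bones, with LOD-selected placement and per-spawn-group offsets.
# Variant, audio reactivity, time of day and front-end state drive color, scale and material parameters; Face also handles BigHeadMode, hidden heads and timeline fading.
# Interpolated position feeds the material as a lagged motion vector.
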NiagaraSystem /BRCosmetics/Characters/Player/Male/Medium/Bodies/M_MED_CanvasPrint/FX/NS_CanvasPrint_Skin_Fire
dependencies {
 usedDataInterfaces [
  &NiagaraDataInterfaceArrayFloat3
  &NiagaraDataInterfaceArrayInt32
  &NiagaraDataInterfaceArrayFloat
  &NiagaraDataInterfaceCurve
 ]
 usedDynamicInputs [
  &VectorLength
  &Waveform
  /Niagara/DynamicInputs/Lerp/SmoothLerpOverTime_Float
  &RemapRange
  &Cosine
  &Add_Float
  &Divide_Float
  &Abs_Float
 ]
 usedModules [
  &self:SystemSpawnScript.SystemScriptSource.SystemScriptGraph.NiagaraNodeAssignment_0.SetVariables_#1
  &self:SystemSpawnScript.SystemScriptSource.SystemScriptGraph.NiagaraNodeAssignment_1.SetVariables_#2
  &self:SystemSpawnScript.SystemScriptSource.SystemScriptGraph.NiagaraNodeAssignment_6.SetVariables_#3
  &SystemState
  &Timeline
  &self:SystemSpawnScript.SystemScriptSource.SystemScriptGraph.NiagaraNodeAssignment_2.SetVariables_#4
  &self:SystemSpawnScript.SystemScriptSource.SystemScriptGraph.NiagaraNodeAssignment_7.SetVariables_#5
  &self:SystemSpawnScript.SystemScriptSource.SystemScriptGraph.NiagaraNodeAssignment_8.SetVariables_#6
  &self:SystemSpawnScript.SystemScriptSource.SystemScriptGraph.NiagaraNodeAssignment_3.SetVariables_#7
 ]
}
emitters {
 Face {
  emitterUpdateScript [
   &EmitterState(
    Inactive Response=::NewEnumerator1
    Life Cycle Mode=::NewEnumerator1
    Loop Behavior=::NewEnumerator1
    Loop Duration Mode=::NewEnumerator1
    UseLoopDelay=false
   )
   &SpawnBurst_Instantaneous(Spawn Count=3, Use Spawn Probability=false, UseLoopCountLimit=false)
  ]
  particleSpawnScript [
   &FN_SetSpawnGroup(Number of Spawn Groups=3)
   &InitializeParticle(
    Mesh Scale=&Lerp_Vector(
     A=&SelectVectorFromArray(
      Direct Array Index=&SpawnGroup()
      Vector Selection Array=DataInterface() {propertyValues {FloatData [<1 0.9 0.8>, <0.9 0.5 0.5>, <8 9 6.3>]}}
     )
     B=&SelectVectorFromArray(
      Direct Array Index=&SpawnGroup()
      Vector Selection Array=DataInterface() {propertyValues {FloatData [<0.9 0.9 0.8>, <0.9 0.5 0.5>, <8 9 6.3>]}}
     )
     Alpha=$User.Variant
    )
    Mesh Scale Mode=::NewEnumerator4
    UsePositionOffset=false
    Write Material Random=true
   )
   &InitialMeshOrientation(
    Mesh Orientation Mode=::NewEnumerator2
    Orientation Vector=&SelectVectorFromArray(
     Direct Array Index=&SpawnGroup()
     Vector Selection Array=DataInterface() {propertyValues {FloatData [<1 0 0>, <1 0 0>, <1 0 0>]}}
    )
    Rotation=&SelectVectorFromArray(
     Direct Array Index=&SpawnGroup()
     Vector Selection Array=DataInterface() {propertyValues {FloatData [<0 0.25 0>, <0 0.25 0>, <0 0.25 0>]}}
    )
    Use Rotation Vector=true
   ) {module=InitialMeshOrientation001}
   /SetVariables(
    Particles.BigMedVisibility:NiagaraInt32=&SelectIntFromArray(Direct Array Index=&SpawnGroup(), Int Selection Array=DataInterface() {propertyValues {IntData [0,0,1]}})
    Particles.GlitchVisibility:NiagaraInt32=&SelectIntFromArray(Direct Array Index=&SpawnGroup(), Int Selection Array=DataInterface() {propertyValues {IntData [0,0,1]}})
    Particles.SmallVisability:NiagaraInt32=&MakeCustomIntFromBool(
     Bool=&SetBoolByFloatComparison(A=0.5, B=$User.Variant)
     True INT=&SelectIntFromArray(Direct Array Index=&SpawnGroup(), Int Selection Array=DataInterface() {propertyValues {IntData [0,0,1]}})
     False INT=&SelectIntFromArray(Direct Array Index=&SpawnGroup(), Int Selection Array=DataInterface() {propertyValues {IntData [1]}})
    )
   ) {isSetParameters=true, module=SetVariables_#8}
   &CameraOffset(Camera Offset Amount=&SelectFloatFromArray(
     Direct Array Index=&SpawnGroup()
     Float Selection Array=DataInterface() {propertyValues {FloatData [0.7,0.3,0]}}
    ))
  ]
  particleUpdateScript [
   &SkeletalMeshLocation(
    Bone / Socket Index=&SelectIntFromArray(Direct Array Index=$User.FaceAccLOD, Int Selection Array=DataInterface() {propertyValues {IntData [0,1]}})
    MaskSpawnGroup=false
    Mesh Sampling Type=::NewEnumerator5
    Sampled Position Offset=&Multiply_VectorByFloat(
     Vector=&SelectVectorFromArray(
      Direct Array Index=&SpawnGroup()
      Vector Selection Array=DataInterface() {propertyValues {FloatData [<8.88 10.5 0>, <0.2653 10.3 0>, <4.4904 15 0>]}}
     )
     Float=&Lerp_Float(A=1, B=1.6, Alpha=$NPC.FortniteNPC.BigHeadMode)
    )
    Skeletal Mesh=$User.HeadRef
    Socket Sampling Mode=::NewEnumerator5
   )
   &ParticleState(Kill Particles When Lifetime Has Elapsed=false, Let Infinitely Lived Particles Die When Emitter Deactivates=true)
   &ScaleMeshSize(Scale Factor=&Multiply_Vector(A=&Multiply_VectorByFloat(Vector=<0.2 0.2 0.2>, Float=0.8), B=<1 1 1>))
   &ScaleMeshSize(Scale Factor=&Multiply_VectorByFloat(
     Vector=&Lerp_Vector(
      A=<0.8 0.8 1>
      B=&SelectVectorFromArray(
       Direct Array Index=&SpawnGroup()
       Vector Selection Array=DataInterface() {propertyValues {FloatData [<0.88 0.88 1.04>, <0.88 0.88 1.04>, <0.8 0.8 1>]}}
      )
      Alpha=$System.AudioReactivity
     )
     Float=&Lerp_Float(A=1, B=1.6, Alpha=$NPC.FortniteNPC.BigHeadMode)
    )) {module=ScaleMeshSize001}
   &ScaleColor(
    Scale Alpha=&Multiply_Float(
     A=&Lerp_Float(
      A=&SelectFloatFromArray(
       Direct Array Index=&SpawnGroup()
       Float Selection Array=DataInterface() {propertyValues {FloatData [3.4,3.4,2]}}
      )
      B=&SelectFloatFromArray(
       Direct Array Index=&SpawnGroup()
       Float Selection Array=DataInterface() {propertyValues {FloatData [1.15,1.15,1]}}
      )
      Alpha=$User.Variant
     )
     B=&FloatFromCurve(FloatCurve=DataInterface() {propertyValues {Curve {keys [<at 0.03 1 RCIM_Cubic>,<at 0.97 0>]}}}, CurveIndex=$System.Timeline.Time)
    )
    Scale RGB=&Multiply_VectorByFloat(
     Vector=&MakeVectorFromLinearColorRGB(Color=&LerpLinearColors(StartColor=<rgba 0.2585 0.1206 0.526 1>, EndColor=<rgba 0 0 0 1>, LerpFactor=$User.Variant))
     Float=&FN_FrontEndFloatModifier(
      Gameplay=&Multiply_Float(A=0.3, B=&FNTimeOfDayScalar(Night Scale=0.8, Morning Scale=0.9, Evening Scale=0.9))
      Front End=0.3
     )
    )
    ScaleA=true
    ScaleRGB=true
   )
   &InterpolateOverTime(
    Rate Of Change:NiagaraFloat=&Multiply_Float(
     A=$System.Interpolate Rate
     B=&SelectFloatFromArray(Direct Array Index=&SpawnGroup(), Float Selection Array=DataInterface() {propertyValues {FloatData [1,1,2]}})
    )
    Target Vector Value:NiagaraPosition=$Particles.Position
    Variable Type=::NewEnumerator2
   )
   /SetVariables(Particles.Interpolate DY:Vector4f=&MakeFloat4FromVectorAndFloat("Vector (RGB)"=$Particles.InterpolateOverTime.InterpolateOverTime_V3_.Moving Average, "Float (Alpha)"=$System.Interpolate Internsity)) {isSetParameters=true, module=SetVariables_#9}
   /SetVariables(Particles.Color1:LinearColor=&MakeLinearColorFromVectorAndFloat(
     "Vector (RGB)"=&MakeVectorFromLinearColorRGB(Color=&LerpLinearColors(
       StartColor=&MultiplyLinearColor(
        Linear Color=&MultiplyLinearColor(
         Linear Color=<rgba 0.7146 0.0637 3 0.097>
         Scale Factor=&MakeVector4_FromFloat(Float=&SelectFloatFromArray(
           Direct Array Index=&SpawnGroup()
           Float Selection Array=DataInterface() {propertyValues {FloatData [1,1,1,1,1,1]}}
          ))
        )
        Scale Factor=&MakeVector4_FromFloat(Float=&FN_FrontEndFloatModifier(
          Gameplay=&Multiply_Float(
           A=&Multiply_Float(A=1, B=&FNTimeOfDayScalar(Night Scale=0.6, Morning Scale=1, Evening Scale=1))
           B=1
          )
          Front End=0.9
         ))
       )
       EndColor=<rgba 0 0 0 1>
       LerpFactor=$User.Variant
      ))
     "Float (Alpha)"=&Multiply_Float(
      A=&Lerp_Float(A=1, B=1.5, Alpha=$User.Variant)
      B=&FloatFromCurve(FloatCurve=DataInterface() {propertyValues {Curve {keys [<at 0.03 1 RCIM_Cubic>,<at 0.97 0>]}}}, CurveIndex=$System.Timeline.Time)
     )
    )) {isSetParameters=true, module=SetVariables_#10}
   /SetVariables(Particles.Color2=&MakeLinearColorFromVectorAndFloat(
     "Vector (RGB)"=&MakeVectorFromLinearColorRGB(Color=&LerpLinearColors(
       StartColor=&MultiplyLinearColor(
        Linear Color=<rgba 4.1657 0.8021 14 1>
        Scale Factor=&MakeVector4_FromFloat(Float=&FN_FrontEndFloatModifier(
          Gameplay=&Multiply_Float(A=0.65, B=&FNTimeOfDayScalar(Night Scale=1, Morning Scale=1, Evening Scale=1.3))
          Front End=0.7
         ))
       )
       EndColor=<rgba 0 0 0 1>
       LerpFactor=$User.Variant
      ))
     "Float (Alpha)"=&FloatFromCurve(FloatCurve=DataInterface() {propertyValues {Curve {keys [<at 0.03 1 RCIM_Cubic>,<at 0.97 0>]}}}, CurveIndex=$System.Timeline.Time)
    )) {isSetParameters=true, module=SetVariables_#11}
   /SetVariables(Particles.Color3:LinearColor=&MakeLinearColorFromVectorAndFloat(
     "Vector (RGB)"=&VectorFromFloat(Value=&FNTimeOfDayScalar(Night Scale=0.65, Morning Scale=0.85, Evening Scale=0.8))
     "Float (Alpha)"=&Multiply_Float(
      A=&FN_FrontEndFloatModifier(Gameplay=0.05, Front End=0.05)
      B=&FloatFromCurve(FloatCurve=DataInterface() {propertyValues {Curve {keys [<at 0.03 1 RCIM_Cubic>,<at 0.97 0>]}}}, CurveIndex=$System.Timeline.Time)
     )
    )) {isSetParameters=true, module=SetVariables_#12}
   &DynamicMaterialParameters(
    Index 0 Param 1=&Multiply_Float(
     A=&Lerp_Float(
      A=&SelectFloatFromArray(
       Direct Array Index=&SpawnGroup()
       Float Selection Array=DataInterface() {propertyValues {FloatData [0.45,0.22,0.4]}}
      )
      B=&SelectFloatFromArray(
       Direct Array Index=&SpawnGroup()
       Float Selection Array=DataInterface() {propertyValues {FloatData [0.54,0.27,0.4]}}
      )
      Alpha=$System.AudioReactivity
     )
     B=&Lerp_Float(A=1, B=2, Alpha=$NPC.FortniteNPC.BigHeadMode)
    )
    Index 0 Param 2=&MakeCustomFloatFromBool(Bool=$System.IsRunning, True Float=1.43, False Float=0.75)
    Index 0 Param 3=6
    Index 0 Param 4=&SelectFloatFromArray(
     Direct Array Index=&SpawnGroup()
     Float Selection Array=DataInterface() {propertyValues {FloatData [0,0.7,0.25,0.5,0.7]}}
    )
    Index 2 Param 1=&Multiply_Float(
     A=$System.AudioReactivity
     B=&SelectFloatFromArray(Direct Array Index=&SpawnGroup(), Float Selection Array=DataInterface() {propertyValues {FloatData [1,1,1]}})
    )
    Index 2 Param 2=&SelectFloatFromArray(Direct Array Index=&SpawnGroup(), Float Selection Array=DataInterface() {propertyValues {FloatData [2,2,1]}})
    Index 2 Param 3=&Lerp_Float(A=1, B=0.55, Alpha=$User.Variant)
    Index 2 Param 4=&Lerp_Float(
     A=&FN_FrontEndFloatModifier(Gameplay=4, Front End=3)
     B=&FN_FrontEndFloatModifier(Gameplay=3, Front End=2.3)
     Alpha=$User.Variant
    )
    Index 3 Param 1=$User.Variant
    Module.Index0Param0Write=true
    Module.Index0Param1Write=true
    Module.Index0Param2Write=true
    Module.Index0Param3Write=true
    Module.Index2Param0Write=true
    Module.Index2Param1Write=true
    Module.Index2Param2Write=true
    Module.Index2Param3Write=true
    Module.Index3Param0Write=true
    Module.Index3Param1Write=true
    Module.Index3Param2Write=true
    Module.Index3Param3Write=true
    Write Parameter Index 2=true
    Write Parameter Index 3=true
   )
   /SetVariables(Particles.GlitchDynamic:Vector4f=&MakeVector4(X=$System.AudioReactivity, Y=0, Z=0, W=0)) {isSetParameters=true}
   &ScaleMeshSize(Scale Factor=&VectorFromFloat(Value=&Lerp_Float(
      A=&SelectFloatFromArray(Direct Array Index=&SpawnGroup(), Float Selection Array=DataInterface() {propertyValues {FloatData [1,1,1]}})
      B=&SelectFloatFromArray(Direct Array Index=&SpawnGroup(), Float Selection Array=DataInterface() {propertyValues {FloatData [1,1,0]}})
      Alpha=$User.Variant
     ))) {module=ScaleMeshSize002}
   &ScaleMeshSize(Scale Factor=&VectorFromFloat(Value=&MakeCustomFloatFromBool(Bool=$User.bisHeadHidden, True Float=0, False Float=1))) {module=ScaleMeshSize003}
  ]
  rendererClasses [&NiagaraMeshRendererProperties]
  renderers [
   {rendererClass=&NiagaraMeshRendererProperties, rendererIndex=0}
   {rendererClass=&NiagaraMeshRendererProperties, rendererIndex=1}
   {rendererClass=&NiagaraMeshRendererProperties, rendererIndex=2}
   {rendererClass=&NiagaraMeshRendererProperties, rendererIndex=3}
  ]
 }
 FireBody {
  emitterUpdateScript [
   &EmitterState(
    Inactive Response=::NewEnumerator1
    Life Cycle Mode=::NewEnumerator1
    Loop Behavior=::NewEnumerator1
    Loop Duration Mode=::NewEnumerator1
    UseLoopDelay=false
   )
   &SpawnBurst_Instantaneous(Spawn Count=5, Use Spawn Probability=false, UseLoopCountLimit=false)
  ]
  particleSpawnScript [
   &FN_SetSpawnGroup(Number of Spawn Groups=5)
   &InitializeParticle(
    Mesh Scale=&SelectVectorFromArray(
     Direct Array Index=&SpawnGroup()
     Vector Selection Array=DataInterface() {propertyValues {FloatData [<0.27 0.06 0.3>, <0.27 0.06 0.3>, <0.27 0.06 0.3>, <1.7 1.7 1>, <1.7 1.7 1>]}}
    )
    Mesh Scale Mode=::NewEnumerator4
    UsePositionOffset=false
    Write Material Random=true
   )
   &InitialMeshOrientation(
    Mesh Orientation Mode=::NewEnumerator2
    Orientation Vector=&Lerp_Vector(
     A=&SelectVectorFromArray(
      Direct Array Index=&SpawnGroup()
      Vector Selection Array=DataInterface() {propertyValues {FloatData [<0 1 0>, <0 1 0>, <0 1 0>, <0.0993 -1.3345 -1.02>, <-0.04 0.83 1.02>]}}
     )
     B=&SelectVectorFromArray(
      Direct Array Index=&SpawnGroup()
      Vector Selection Array=DataInterface() {propertyValues {FloatData [<1 0.0473 0>, <0 1 0>, <0 1 0>, <0.0993 -1.3345 -1.02>, <-0.04 0.83 1.02>]}}
     )
     Alpha=&ClampFloat(Float=&MakeFloatFromInt(Int=$User.BodyAccLOD))
    )
    Rotation=&Lerp_Vector(
     A=&SelectVectorFromArray(
      Direct Array Index=&SpawnGroup()
      Vector Selection Array=DataInterface() {propertyValues {FloatData [<0.065 -0.015 -0.0411>, <0.065 0.0189 0>, <0.065 0 -0.0473>, <0.5497 -0.2771 -0.045>, <0.2719 0.009 0>]}}
     )
     B=&SelectVectorFromArray(
      Direct Array Index=&SpawnGroup()
      Vector Selection Array=DataInterface() {propertyValues {FloatData [<0.2973 0.295 0>, <0 -0.0072 0.018>, <0 0.0168 0>, <0.5497 -0.2771 -0.045>, <0.2719 0.009 0>]}}
     )
     Alpha=&ClampFloat(Float=&MakeFloatFromInt(Int=$User.BodyAccLOD))
    )
    Use Rotation Vector=true
   )
   &InitialMeshOrientation(
    Mesh Orientation Mode=::NewEnumerator2
    Orientation Vector=&SelectVectorFromArray(
     Direct Array Index=&SpawnGroup()
     Vector Selection Array=DataInterface() {propertyValues {FloatData [<1 0 0>, <0 1 0>, <0 1 0>, <0.0993 -1.3345 -1.02>, <-0.04 0.83 1.02>]}}
    )
    Rotation=&SelectVectorFromArray(
     Direct Array Index=&SpawnGroup()
     Vector Selection Array=DataInterface() {propertyValues {FloatData [<0.2973 0.3221 0>, <0 -0.0072 0.018>, <0 0.0168 0>, <0.5497 -0.2771 -0.045>, <0.2719 0.009 0>]}}
    )
    Use Rotation Vector=true
   ) {enabled=false, module=InitialMeshOrientation001}
   /SetVariables(Particles.MeshIndexPicker:NiagaraInt32=&SelectIntFromArray(Direct Array Index=&SpawnGroup(), Int Selection Array=DataInterface() {propertyValues {IntData [0,0,0,1,1]}})) {isSetParameters=true, module=SetVariables_#8}
  ]
  particleUpdateScript [
   &SkeletalMeshLocation(
    Bone / Socket Index=&MakeIntFromFloat(Float=&Lerp_Float(
      A=&MakeFloatFromInt(Int=&SelectIntFromArray(Direct Array Index=&SpawnGroup(), Int Selection Array=DataInterface() {propertyValues {IntData [0,1,2,4,5]}}))
      B=&MakeFloatFromInt(Int=&SelectIntFromArray(Direct Array Index=&SpawnGroup(), Int Selection Array=DataInterface() {propertyValues {IntData [3,3,3,4,5]}}))
      Alpha=&ClampFloat(Float=&MakeFloatFromInt(Int=$User.BodyAccLOD))
     ))
    Bone Sampling Mode=::NewEnumerator4
    MaskSpawnGroup=false
    Sampled Position Offset=&Lerp_Vector(
     A=&SelectVectorFromArray(
      Direct Array Index=&SpawnGroup()
      Vector Selection Array=DataInterface() {propertyValues {FloatData [<7.3 -0.5 1.1246>, <7.8695 0 1.7131>, <7.5 -2 2.0495>, <-19 3 1>, <19 -3 -1>]}}
     )
     B=&SelectVectorFromArray(
      Direct Array Index=&SpawnGroup()
      Vector Selection Array=DataInterface() {propertyValues {FloatData [<-6.0742 -7.7 23.5>, <-5.5846 4.1465 -25.1787>, <-5.3344 -7.5613 -23.1915>, <-19 3 1>, <19 -3 -1>]}}
     )
     Alpha=&ClampFloat(Float=&MakeFloatFromInt(Int=$User.BodyAccLOD))
    )
    Skeletal Mesh=$User.BodyRef
   )
   &SkeletalMeshLocation(
    Bone / Socket Index=&SelectIntFromArray(Direct Array Index=&SpawnGroup(), Int Selection Array=DataInterface() {propertyValues {IntData [1,1,1,2,3]}})
    Bone Sampling Mode=::NewEnumerator4
    MaskSpawnGroup=false
    Sampled Position Offset=&SelectVectorFromArray(
     Direct Array Index=&SpawnGroup()
     Vector Selection Array=DataInterface() {propertyValues {FloatData [<-6.0742 -7.7 23.5>, <-5.5846 4.1465 -25.1787>, <-5.3344 -7.5613 -23.1915>, <-19 3 1>, <19 -3 -1>]}}
    )
    Skeletal Mesh=$System.Skeletal Mesh
   ) {enabled=false, module=SkeletalMeshLocation001}
   &ParticleState(Kill Particles When Lifetime Has Elapsed=false, Let Infinitely Lived Particles Die When Emitter Deactivates=true)
   &ScaleMeshSize(Scale Factor=&Multiply_Vector(A=&Multiply_VectorByFloat(Vector=<0.2 0.2 0.2>, Float=0.8), B=<1 1 1>))
   &ScaleMeshSize(Scale Factor=&Lerp_Vector(
     A=<0.8 0.8 1>
     B=&SelectVectorFromArray(
      Direct Array Index=&SpawnGroup()
      Vector Selection Array=DataInterface() {propertyValues {FloatData [<0.8 0.8 1>]}}
     )
     Alpha=$System.AudioReactivity
    )) {module=ScaleMeshSize001}
   &ScaleColor(
    Scale Alpha=&Lerp_Float(
     A=&SelectFloatFromArray(Direct Array Index=&SpawnGroup(), Float Selection Array=DataInterface() {propertyValues {FloatData [2.6]}})
     B=&SelectFloatFromArray(Direct Array Index=&SpawnGroup(), Float Selection Array=DataInterface() {propertyValues {FloatData [1.5]}})
     Alpha=$User.Variant
    )
    Scale RGB=&Multiply_VectorByFloat(
     Vector=&MakeVectorFromLinearColorRGB(Color=&LerpLinearColors(StartColor=<rgba 0.2585 0.1206 0.526 1>, EndColor=<rgba 0 0 0 1>, LerpFactor=$User.Variant))
     Float=&FN_FrontEndFloatModifier(
      Gameplay=&Multiply_Float(A=0.3, B=&FNTimeOfDayScalar(Night Scale=0.8, Morning Scale=0.9, Evening Scale=0.9))
      Front End=0.3
     )
    )
    ScaleA=true
    ScaleRGB=true
   ) {module=ScaleColor001}
   &InterpolateOverTime(
    Rate Of Change:NiagaraFloat=&Multiply_Float(
     A=$System.Interpolate Rate
     B=&SelectFloatFromArray(Direct Array Index=&SpawnGroup(), Float Selection Array=DataInterface() {propertyValues {FloatData [2]}})
    )
    Target Vector Value:NiagaraPosition=$Particles.Position
    Variable Type=::NewEnumerator2
   )
   /SetVariables(Particles.Interpolate DY:Vector4f=&MakeFloat4FromVectorAndFloat(
     "Vector (RGB)"=$Particles.InterpolateOverTime.InterpolateOverTime_V3_.Moving Average
     "Float (Alpha)"=&SelectFloatFromArray(
      Direct Array Index=&SpawnGroup()
      Float Selection Array=DataInterface() {propertyValues {FloatData [0,0,0,0.8,0.8]}}
     )
    )) {isSetParameters=true, module=SetVariables_#9}
   /SetVariables(Particles.Color1:LinearColor=&MakeLinearColorFromVectorAndFloat(
     "Vector (RGB)"=&MakeVectorFromLinearColorRGB(Color=&LerpLinearColors(
       StartColor=&MultiplyLinearColor(
        Linear Color=&MultiplyLinearColor(
         Linear Color=<rgba 0.9528 0.0849 4 0.097>
         Scale Factor=&MakeVector4_FromFloat(Float=&SelectFloatFromArray(
           Direct Array Index=&SpawnGroup()
           Float Selection Array=DataInterface() {propertyValues {FloatData [1,1,1,1,1,1]}}
          ))
        )
        Scale Factor=&MakeVector4_FromFloat(Float=&FN_FrontEndFloatModifier(
          Gameplay=&Multiply_Float(
           A=&Multiply_Float(A=1, B=&FNTimeOfDayScalar(Night Scale=0.6, Morning Scale=1, Evening Scale=1))
           B=1
          )
          Front End=0.9
         ))
       )
       EndColor=<rgba 0 0 0 1>
       LerpFactor=$User.Variant
      ))
     "Float (Alpha)"=&Lerp_Float(A=1, B=1, Alpha=$User.Variant)
    )) {isSetParameters=true, module=SetVariables_#10}
   /SetVariables(Particles.Color2=&MakeLinearColorFromVectorAndFloat(
     "Vector (RGB)"=&MakeVectorFromLinearColorRGB(Color=&LerpLinearColors(
       StartColor=&MultiplyLinearColor(
        Linear Color=<rgba 4.1657 0.8021 14 1>
        Scale Factor=&MakeVector4_FromFloat(Float=&FN_FrontEndFloatModifier(
          Gameplay=&Multiply_Float(A=0.65, B=&FNTimeOfDayScalar(Night Scale=1, Morning Scale=1, Evening Scale=1.3))
          Front End=0.7
         ))
       )
       EndColor=<rgba 0 0 0 1>
       LerpFactor=$User.Variant
      ))
     "Float (Alpha)"=1
    )) {isSetParameters=true, module=SetVariables_#11}
   /SetVariables(Particles.Color3:LinearColor=&MakeLinearColorFromVectorAndFloat(
     "Vector (RGB)"=&VectorFromFloat(Value=&FNTimeOfDayScalar(Night Scale=0.65, Morning Scale=0.85, Evening Scale=0.8))
     "Float (Alpha)"=&FN_FrontEndFloatModifier(Gameplay=0.05, Front End=0.05)
    )) {isSetParameters=true, module=SetVariables_#12}
   &DynamicMaterialParameters(
    Index 0 Param 1=&Lerp_Float(
     A=&SelectFloatFromArray(
      Direct Array Index=&SpawnGroup()
      Float Selection Array=DataInterface() {propertyValues {FloatData [0.16,0.16,0.16,0.54]}}
     )
     B=&SelectFloatFromArray(
      Direct Array Index=&SpawnGroup()
      Float Selection Array=DataInterface() {propertyValues {FloatData [0.22,0.22,0.22,0.62]}}
     )
     Alpha=$System.AudioReactivity
    )
    Index 0 Param 2=&MakeCustomFloatFromBool(Bool=$System.IsRunning, True Float=1.43, False Float=0.75)
    Index 0 Param 3=6
    Index 0 Param 4=&SelectFloatFromArray(
     Direct Array Index=&SpawnGroup()
     Float Selection Array=DataInterface() {propertyValues {FloatData [0,0.7,0.25,0.5,0.3]}}
    )
    Index 2 Param 1=&Multiply_Float(
     A=$System.AudioReactivity
     B=&SelectFloatFromArray(Direct Array Index=&SpawnGroup(), Float Selection Array=DataInterface() {propertyValues {FloatData [1,1,1]}})
    )
    Index 2 Param 2=&SelectFloatFromArray(
     Direct Array Index=&SpawnGroup()
     Float Selection Array=DataInterface() {propertyValues {FloatData [0.1,0.1,0.1,13,13]}}
    )
    Index 2 Param 3=&Lerp_Float(A=1, B=0.55, Alpha=$User.Variant)
    Index 2 Param 4=&Lerp_Float(
     A=&FN_FrontEndFloatModifier(Gameplay=7, Front End=7)
     B=&FN_FrontEndFloatModifier(Gameplay=3, Front End=2.3)
     Alpha=$User.Variant
    )
    Index 3 Param 1=$User.Variant
    Module.Index0Param0Write=true
    Module.Index0Param1Write=true
    Module.Index0Param2Write=true
    Module.Index0Param3Write=true
    Module.Index2Param0Write=true
    Module.Index2Param1Write=true
    Module.Index2Param2Write=true
    Module.Index2Param3Write=true
    Module.Index3Param0Write=true
    Module.Index3Param1Write=true
    Module.Index3Param2Write=true
    Module.Index3Param3Write=true
    Write Parameter Index 2=true
    Write Parameter Index 3=true
   )
  ]
  rendererClasses [&NiagaraMeshRendererProperties]
  renderers [
   {rendererClass=&NiagaraMeshRendererProperties, rendererIndex=0}
   {rendererClass=&NiagaraMeshRendererProperties, rendererIndex=1}
  ]
 }
}
summary {
 systemName=NS_CanvasPrint_Skin_Fire
 emitters [
  {emitterName=FireBody, rendererClasses [&NiagaraMeshRendererProperties]}
  {emitterName=Face, rendererClasses [&NiagaraMeshRendererProperties]}
 ]
 userVariables [
  User.bIsADS {type=NiagaraBool}
  User.bIsFrontEnd {type=NiagaraBool}
  User.bIsFrontEndPreview {type=NiagaraBool}
  User.bisHeadHidden {type=NiagaraBool}
  User.BodyAccLOD {type=NiagaraInt32}
  User.BodyRef {
   type=NiagaraDataInterfaceSkeletalMesh
   defaultValue {
    struct=NiagaraExt_VariableValue_DataInterface
    value {dataInterface=&self:NiagaraDataInterfaceSkeletalMesh_1}
   }
  }
  User.Emote begin {type=NiagaraBool}
  User.FaceAccLOD {type=NiagaraInt32}
  User.HeadRef {
   type=NiagaraDataInterfaceSkeletalMesh
   defaultValue {
    struct=NiagaraExt_VariableValue_DataInterface
    value {dataInterface=&self:NiagaraDataInterfaceSkeletalMesh_0}
   }
  }
  User.IsSkydiving {type=NiagaraBool}
  User.Variant {type=NiagaraFloat, defaultValue=0, description=""}
 ]
}
