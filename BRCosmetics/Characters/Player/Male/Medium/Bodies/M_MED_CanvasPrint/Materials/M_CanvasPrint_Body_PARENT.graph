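# Parent material for the CanvasPrint character body, adding audio-reactive glow to shoulder feathers and body between the shared character PreFX and PostFX functions.
# Audio spectrum, with a cosine-based fake frontend loop, drives feather-mask erosion and multiplication, shaped by feather AO.
# Feather color shifts with view angle around a time-rotated axis, with amplitude changing fringe falloff.
# Static switches gate emissive (default on, time-of-day scaled) and base-color tint (default off).
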Material /BRCosmetics/Characters/Player/Male/Medium/Bodies/M_MED_CanvasPrint/Materials/M_CanvasPrint_Body_PARENT
graph {
 properties {
  bUseMaterialAttributes=true
  bUsedWithMeshDeformer=true
  bUsedWithMorphTargets=true
  bUsedWithSkeletalMesh=true
  pixelDepthOffsetMode=PDOM_Legacy
  shadingModel=MSM_FromMaterialExpression
  usedShadingModels=Default Lit | Subsurface | Clear Coat | Subsurface Profile | Cloth
 }
 nodes [
  n1 = MaterialFunctionCall() {materialFunction=&MF_CharacterParent_2023_PreFX}
  n2 = GetMaterialAttributes(None=@n1.Result) {attributeGetTypes [#1, #2]}
  n3 = TextureSampleParameter2D() {
   group=CanvasPrint
   parameterName=FeatherMask
   samplerType=SAMPLERTYPE_LinearColor
   texture=/BRCosmetics/Characters/Player/Male/Medium/Bodies/M_MED_CanvasPrint/Textures/T_CanvasPrint_Body_TechArtFX_Mask
  }
  n4 = Divide(A=Time(), B=ScalarParameter() {defaultValue=1.3, group=Audio Reactivity, parameterName=FrontendFakeAudio_LoopLength, sortPriority=200})
  MaterialExpressionMultiply_21 = Multiply(
   A=ConstantBiasScale(None=Cosine(None=@n4))
   B=ScalarParameter() {defaultValue=0.5, group=Audio Reactivity, parameterName=FrontendFakeAudio_Amplitude, sortPriority=200}
  )
  n5 = Convert(None=@MaterialExpressionMultiply_21) {
   convertInputs [{
     type=Scalar
     expressionInput {
      expression=&self:MaterialExpressionMultiply_21
      inputName=None
      mask=0
      maskA=0
      maskB=0
      maskG=0
      maskR=0
      outputIndex=0
     }
    }]
   convertOutputs [{type=Vector4, defaultValue=<rgba 0 0 0 1>}]
  }
  n6 = MaterialFunctionCall!(InAmplitude, InSpectrum, CustomFakeSpectrum=@n5, CustomFakeAmplitude=@MaterialExpressionMultiply_21) {materialFunction=/BRCosmetics/Global/MaterialFunctions/AudioReactivity/MF_AudioReactivity_FrontendLoop}
  n7 = Power(
   Base=Max(A=ComponentMask(None=@n6.Spectrum) {a=true}, B=ComponentMask(None=@n6.Spectrum) {g=true})
   Exp=ScalarParameter() {defaultValue=4, group=CanvasPrint, parameterName=Spectrum_Exp}
  ) {constExponent=4}
  n8 = Power(Base=OneMinus(None=@n3.G), Exp=ScalarParameter() {defaultValue=6, group=CanvasPrint, parameterName=MaskErosion_Exp}) {constExponent=3}
  n9 = Power(Base=Saturate(None=Subtract(A=@n7, B=@n8)))
  n10 = Multiply(A=@n7, B=Power(Base=@n3.G, Exp=ScalarParameter() {defaultValue=3, group=CanvasPrint, parameterName=MaskMultiply_Exp}))
  n11 = ScalarParameter() {defaultValue=0.8, group=CanvasPrint, parameterName=ErosionBlend, sliderMax=1}
  n12 = Multiply(A=@n3.G, B=Saturate(None=LinearInterpolate(A=@n9, B=@n10, Alpha=@n11)))
  n13 = ScalarParameter() {defaultValue=1, group=CanvasPrint, parameterName=FeathersAO_Exp}
  n14 = LinearInterpolate(
   A=@n12
   B=Multiply(A=Power(Base=@n3.R, Exp=@n13), B=@n12)
   Alpha=ScalarParameter() {defaultValue=1, group=CanvasPrint, parameterName=FeathersAO_Influence}
  )
  n15 = ScalarParameter() {defaultValue=15, group=CanvasPrint, parameterName=FeathersReactivity_Strength}
  n16 = LinearInterpolate(
   A=@n2.Base Color
   B=Multiply(A=@n2.Base Color, B=VectorParameter() {group=CanvasPrint, parameterName=Reactive_BaseColorTint}.RGB)
   Alpha=Saturate(None=Multiply(A=@n14, B=@n15))
  )
  n17 = Multiply(A=@n2.Emissive Color, B=ScalarParameter() {defaultValue=15, group=CanvasPrint, parameterName=BodyGlowReactivity_Strength}) {constB=10}
  n18 = NamedRerouteUsage()
  n19 = Multiply(A=@n15, B=NamedRerouteUsage())
  n20 = Ceil(None=@n3.G)
  n21 = Multiply(
   A=LinearInterpolate(
    A=LinearInterpolate(A=@n2.Emissive Color, B=@n17, Alpha=@n18)
    B=Multiply(A=@n14, B=@n19)
    Alpha=NamedRerouteDeclaration(None=@n20) {name=ShouldersMask, nodeColor=<rgba 1 0 0.3281 1>}
   )
   B=LinearInterpolate(
    B=ScalarParameter() {defaultValue=0.8, parameterName=BrightnessScale_Night}
    Alpha=MaterialFunctionCall() {materialFunction=&Scalar_Time_of_Day}.Result
   ) {constA=1, constB=0.2}
  )
  MaterialFunctionCall(
   Material=SetMaterialAttributes!(
    MaterialAttributes=@n2.MaterialAttributes
    Base Color=StaticSwitchParameter(True=@n16, False=@n2.Base Color) {group=CanvasPrint, parameterName=Reactive_BaseColor, sortPriority=0}
    Emissive Color=StaticSwitchParameter(True=@n21, False=@n2.Emissive Color) {defaultValue=true, group=CanvasPrint, parameterName=Reactive_Emissive, sortPriority=0}
   ) {attributeSetTypes [#1, #2]}
   M=@n1.M
   Pre Skinned Local Position=@n1.M
   Distance Blend=@n1.M
   Fresnel=@n1.M
  ) {materialFunction=&MF_CharacterParent_2023_PostFX}
  n22 = VectorParameter() {group=CanvasPrint, parameterName=ShoulderFeatherColor_M, defaultValue=<rgba 1 0 1 1>}
  n23 = Normalize(VectorInput=Transform(None=Constant3Vector() {constant=<rgba 1 0 0 0>}))
  n24 = VertexNormalWS()
  n25 = Multiply(A=Time()) {constB=0.2}
  n26 = Constant3Vector()
  n27 = DotProduct(A=CameraVectorWS(), B=Add(A=@n23, B=RotateAboutAxis(NormalizedRotationAxis=@n24, RotationAngle=@n25, PivotPoint=@n26, Position=@n23)))
  n28 = Power(Base=@n6.Amplitude, Exp=ScalarParameter() {defaultValue=4, group=CanvasPrint, parameterName=Amplitude_Exp}) {constExponent=4}
  n29 = Power(
   Base=Abs(None=@n27)
   Exp=LinearInterpolate(
    A=ScalarParameter() {defaultValue=6, group=CanvasPrint, parameterName=FeatherFringe_Exp_Quiet}
    B=ScalarParameter() {defaultValue=3, group=CanvasPrint, parameterName=FeatherFringe_Exp_Loud}
    Alpha=NamedRerouteDeclaration(None=@n28) {name=Amplitude, nodeColor=<rgba 0.6172 0 1 1>}
   ) {constA=6, constB=3}
  )
  n30 = OneMinus(None=@n29)
  Multiply(A=@n22.RGB, B=@n30)
  Multiply(A=@n3.R, B=@n29)
  n31 = OneMinus(None=@n27)
  n32 = Power(Base=Abs(None=@n27))
  n33 = Multiply(
   A=Multiply(A=VectorParameter() {group=CanvasPrint, parameterName=ShoulderFeatherColor_L, defaultValue=<rgba 0 0.25 1 1>}.RGB, B=Saturate(None=@n31)) {constB=-1}
   B=@n32
  )
  n34 = Multiply(
   A=Multiply(A=Saturate(None=@n27), B=VectorParameter() {group=CanvasPrint, parameterName=ShoulderFeatherColor_R, defaultValue=<rgba 1 0 0.25 1>}.RGB)
   B=@n32
  )
  n35 = ConstantBiasScale(None=@n27)
  NamedRerouteDeclaration(None=LinearInterpolate(A=LinearInterpolate(A=@n33, B=@n34, Alpha=@n35), B=@n22.RGB, Alpha=@n30) {constB=0}) {name=ReactiveColor, nodeColor=<rgba 0 1 0.4844 1>}
 ]
}
