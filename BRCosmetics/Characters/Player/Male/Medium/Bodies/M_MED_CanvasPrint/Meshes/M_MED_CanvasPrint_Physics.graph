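# Physics asset for the M_MED_CanvasPrint male player body cosmetic, giving its four dynamic wings (front and back, left and right) capsule bodies.
# Each wing's segments 2–7 constrain directly to that wing's segment 1 rather than chaining, with tight swing and twist limits.
# The head, upper arms and right clavicle also carry capsules without constraints; the left clavicle has none.
# The back-left segment 3 constraint appears twice.
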PhysicsAsset /BRCosmetics/Characters/Player/Male/Medium/Bodies/M_MED_CanvasPrint/Meshes/M_MED_CanvasPrint_Physics
bodies {
 clavicle_r [{shapeType=Capsule, length=8.5, radius=6.5, shapeName=clavicle_r_capsule, center=<-11.0527 0 0>, rotation=<rot -80 179.9995 -179.9997>}]
 dyn_wing_back_l_1 [{shapeType=Capsule, length=8.8, radius=2.9, shapeName=dyn_wing_back_l_1_capsule, center=<4.7063 0 0.4118>, rotation=<rot -85 -0.0001 -0.0004>}]
 dyn_wing_back_l_2 [{
   shapeType=Capsule
   length=10.2
   radius=1.5
   shapeName=dyn_wing_back_l_2_capsule
   center=<5.0299 -0.4661 0.5205>
   rotation=<rot -85.0001 -0.0001 -0.0003>
  }]
 dyn_wing_back_l_3 [
  {
   shapeType=Capsule
   length=9
   radius=0.6
   shapeName=dyn_wing_back_l_3_capsule
   center=<2.9797 0.3532 0.2004>
   rotation=<rot -79.9999 -0.0003 -0.0005>
  }
  {
   shapeType=Capsule
   length=8.6
   radius=0.6
   shapeName=dyn_wing_back_l_3_capsule
   center=<3.4681 -0.6345 -1.3069>
   rotation=<rot -90 178.2101 -172.5853>
  }
 ]
 dyn_wing_back_l_4 [{
   shapeType=Capsule
   length=8.2
   radius=0.6
   shapeName=dyn_wing_back_l_4_capsule
   center=<2.9999 0.5143 1.1395>
   rotation=<rot -67.1903 0.3558 2.4805>
  }]
 dyn_wing_back_l_5 [{
   shapeType=Capsule
   length=6.8
   radius=0.7
   shapeName=dyn_wing_back_l_5_capsule
   center=<3.7084 1.4117 -2.5384>
   rotation=<rot -64.7958 -166.4756 175.9398>
  }]
 dyn_wing_back_l_6 [{
   shapeType=Capsule
   length=5
   radius=1.1
   shapeName=dyn_wing_back_l_6_capsule
   center=<2.1068 0.0015 -0.5362>
   rotation=<rot 65.0235 -3.0632 -2.9469>
  }]
 dyn_wing_back_l_7 [{
   shapeType=Capsule
   length=6
   radius=1.1
   shapeName=dyn_wing_back_l_7_capsule
   center=<2.1864 0 -0.125>
   rotation=<rot -80.0001 179.9991 -179.9994>
  }]
 dyn_wing_back_r_1 [{
   shapeType=Capsule
   length=8.8
   radius=2.9
   shapeName=dyn_wing_back_l_1_capsule
   center=<-4.0607 0 -0.3553>
   rotation=<rot -85 -0.0001 179.9996>
  }]
 dyn_wing_back_r_2 [{
   shapeType=Capsule
   length=12.9
   radius=1.1
   shapeName=dyn_wing_back_l_2_capsule
   center=<-3.7366 0.6791 -0.0143>
   rotation=<rot -85.0003 0.0012 -175.0016>
  }]
 dyn_wing_back_r_3 [
  {
   shapeType=Capsule
   length=7
   radius=1.1
   shapeName=dyn_wing_back_l_3_capsule
   center=<-4.1426 -0.3613 -0.9711>
   rotation=<rot -84.9999 0.0005 -175.0018>
  }
  {
   shapeType=Capsule
   length=8.2
   radius=1.1
   shapeName=dyn_wing_back_r_3_capsule
   center=<-4.1479 0.6594 1.1113>
   rotation=<rot -82.7461 175.8885 -165.3092>
  }
 ]
 dyn_wing_back_r_4 [{
   shapeType=Capsule
   length=8.2
   radius=0.6
   shapeName=dyn_wing_back_l_4_capsule
   center=<-2.5481 -0.005 -1.8815>
   rotation=<rot -72.0359 57.0535 124.2657>
  }]
 dyn_wing_back_r_5 [{
   shapeType=Capsule
   length=7.8
   radius=1.2
   shapeName=dyn_wing_back_l_5_capsule
   center=<-2.8733 -1.2417 1.9956>
   rotation=<rot -65.093 178.8725 10.8949>
  }]
 dyn_wing_back_r_6 [{
   shapeType=Capsule
   length=5
   radius=1.1
   shapeName=dyn_wing_back_l_6_capsule
   center=<-1.985 -0.008 0.798>
   rotation=<rot 65.0235 -3.0632 177.0531>
  }]
 dyn_wing_back_r_7 [{
   shapeType=Capsule
   length=6
   radius=1.1
   shapeName=dyn_wing_back_l_7_capsule
   center=<-2.1345 -0.4196 -0.235>
   rotation=<rot -80.0001 -170.0009 0.0005>
  }]
 dyn_wing_front_l_1 [{shapeType=Capsule, length=7.4, radius=3.2, shapeName=dyn_wing_front_l_1_capsule, center=<5.1396 0 0>, rotation=<rot 90 -32.5436 -22.5439>}]
 dyn_wing_front_l_2 [{
   shapeType=Capsule
   length=13.8
   radius=1.5
   shapeName=dyn_wing_front_l_2_capsule
   center=<4.1866 0 -0.7382>
   rotation=<rot 72.8121 -0.0002 -0.0007>
  }]
 dyn_wing_front_l_3 [
  {
   shapeType=Capsule
   length=10.3
   radius=1
   shapeName=dyn_wing_front_l_3_capsule
   center=<3.2027 -2.0751 -1.721>
   rotation=<rot -69.9998 -180 -179.9998>
  }
  {
   shapeType=Capsule
   length=8.9
   radius=0.9
   shapeName=dyn_wing_front_l_3_capsule
   center=<6.3187 -2.8236 0.7704>
   rotation=<rot 61.8749 -89.9992 -81.562>
  }
 ]
 dyn_wing_front_l_4 [
  {
   shapeType=Capsule
   length=10
   radius=0.6
   shapeName=dyn_wing_front_l_4_capsule
   center=<4.5985 -0.2624 -1.4356>
   rotation=<rot 67.1869 -0.0003 2.8118>
  }
  {
   shapeType=Capsule
   length=8.4
   radius=0.6
   shapeName=dyn_wing_front_l_4_capsule
   center=<3.7596 2.2573 -1.1598>
   rotation=<rot 72.2685 2.3896 -12.7067>
  }
 ]
 dyn_wing_front_l_5 [{
   shapeType=Capsule
   length=9.1
   radius=0.8
   shapeName=dyn_wing_front_l_5_capsule
   center=<3.7394 0.7613 0.0354>
   rotation=<rot -79.9998 179.9994 -164.9988>
  }]
 dyn_wing_front_l_6 [
  {
   shapeType=Capsule
   length=10.1
   radius=0.7
   shapeName=dyn_wing_front_l_6_capsule
   center=<3.4451 -1.9237 0>
   rotation=<rot -90 139.7086 -159.7091>
  }
  {
   shapeType=Capsule
   length=3.9
   radius=0.8
   shapeName=dyn_wing_front_l_6_capsule
   center=<2.6308 0 -1.8081>
   rotation=<rot 54.9998 -0.0003 -0.0003>
  }
 ]
 dyn_wing_front_l_7 [{
   shapeType=Capsule
   length=8.5
   radius=0.7
   shapeName=dyn_wing_front_l_7_capsule
   center=<3.4318 0.4866 0.5016>
   rotation=<rot -81.5627 5.001 -0.0016>
  }]
 dyn_wing_front_r_1 [{shapeType=Capsule, length=7.4, radius=3.2, shapeName=dyn_wing_front_l_1_capsule, center=<-5.1151 0.8475 0>, rotation=<rot 90 0 -165.0004>}]
 dyn_wing_front_r_2 [{
   shapeType=Capsule
   length=12.2
   radius=1.2
   shapeName=dyn_wing_front_l_2_capsule
   center=<-6.2446 0.3418 1.3487>
   rotation=<rot 72.8118 -0.0006 179.9988>
  }]
 dyn_wing_front_r_3 [
  {
   shapeType=Capsule
   length=10.7
   radius=0.6
   shapeName=dyn_wing_front_l_3_capsule
   center=<-4.0318 2.25 1.9363>
   rotation=<rot -65.0229 177.946 -3.0621>
  }
  {
   shapeType=Capsule
   length=12
   radius=0.5
   shapeName=dyn_wing_front_r_3_capsule
   center=<-6.0286 2.6993 -0.8937>
   rotation=<rot -90 0 -9.9999>
  }
 ]
 dyn_wing_front_r_4 [
  {
   shapeType=Capsule
   length=9.4
   radius=1.5
   shapeName=dyn_wing_front_l_4_capsule
   center=<-3.9522 0.2987 1.5286>
   rotation=<rot 69.3745 -0.0003 179.9994>
  }
  {
   shapeType=Capsule
   length=8.8
   radius=0.5
   shapeName=dyn_wing_front_r_4_capsule
   center=<-4.049 -2.3767 1.5709>
   rotation=<rot 74.9992 -0.0008 -10.0013>
  }
 ]
 dyn_wing_front_r_5 [{
   shapeType=Capsule
   length=7.8
   radius=0.7
   shapeName=dyn_wing_front_l_5_capsule
   center=<-3.6149 -1.4654 -0.1246>
   rotation=<rot -70.129 -168.6338 -1.1943>
  }]
 dyn_wing_front_r_6 [
  {
   shapeType=Capsule
   length=8.7
   radius=1.3
   shapeName=dyn_wing_front_l_6_capsule
   center=<-4.1363 2.1596 -0.1818>
   rotation=<rot -90 180 -20.0005>
  }
  {
   shapeType=Capsule
   length=4.2
   radius=1.1
   shapeName=dyn_wing_front_r_6_capsule
   center=<-2.0798 -0.104 2.0483>
   rotation=<rot 64.9998 -0.0004 9.9996>
  }
 ]
 dyn_wing_front_r_7 [{
   shapeType=Capsule
   length=5.6
   radius=1.2
   shapeName=dyn_wing_front_l_7_capsule
   center=<-3.8519 -0.0546 -0.2901>
   rotation=<rot -90 180 5.6246>
  }]
 head [{shapeType=Capsule, length=5.9, radius=14.7, shapeName=head_capsule, center=<9.4098 1.3958 0>, rotation=<rot 90 -178.8065 172.7563>}]
 upperarm_l [{shapeType=Capsule, length=19.2, radius=5.8, shapeName=upperarm_l_capsule, center=<12.2691 0 -1.5173>, rotation=<rot -90 179.9209 -179.9209>}]
 upperarm_r [{
   shapeType=Capsule
   length=19.2
   radius=5.8
   shapeName=upperarm_r_capsule
   center=<-13.6203 1.2024 1.5421>
   rotation=<rot -90 91.2733 -85.6484>
  }]
}
constraints [
 {bone1Name=dyn_wing_back_l_2, bone2Name=dyn_wing_back_l_1, swing1LimitDegrees=15, swing2LimitDegrees=15, twistLimitDegrees=10}
 {bone1Name=dyn_wing_back_l_3, bone2Name=dyn_wing_back_l_1, swing1LimitDegrees=15, swing2LimitDegrees=15, twistLimitDegrees=10}
 {bone1Name=dyn_wing_back_l_3, bone2Name=dyn_wing_back_l_1, swing1LimitDegrees=15, swing2LimitDegrees=15, twistLimitDegrees=10}
 {bone1Name=dyn_wing_back_l_4, bone2Name=dyn_wing_back_l_1, swing1LimitDegrees=15, swing2LimitDegrees=15, twistLimitDegrees=10}
 {bone1Name=dyn_wing_back_l_5, bone2Name=dyn_wing_back_l_1, swing1LimitDegrees=15, swing2LimitDegrees=15, twistLimitDegrees=10}
 {bone1Name=dyn_wing_back_l_6, bone2Name=dyn_wing_back_l_1, swing1LimitDegrees=15, swing2LimitDegrees=15, twistLimitDegrees=10}
 {bone1Name=dyn_wing_back_l_7, bone2Name=dyn_wing_back_l_1, swing1LimitDegrees=15, swing2LimitDegrees=15, twistLimitDegrees=10}
 {bone1Name=dyn_wing_front_l_2, bone2Name=dyn_wing_front_l_1, swing1LimitDegrees=15, swing2LimitDegrees=15, twistLimitDegrees=10}
 {bone1Name=dyn_wing_front_l_3, bone2Name=dyn_wing_front_l_1, swing1LimitDegrees=15, swing2LimitDegrees=15, twistLimitDegrees=10}
 {bone1Name=dyn_wing_front_l_4, bone2Name=dyn_wing_front_l_1, swing1LimitDegrees=15, swing2LimitDegrees=15, twistLimitDegrees=10}
 {bone1Name=dyn_wing_front_l_5, bone2Name=dyn_wing_front_l_1, swing1LimitDegrees=15, swing2LimitDegrees=15, twistLimitDegrees=10}
 {bone1Name=dyn_wing_front_l_6, bone2Name=dyn_wing_front_l_1, swing1LimitDegrees=15, swing2LimitDegrees=15, twistLimitDegrees=10}
 {bone1Name=dyn_wing_front_l_7, bone2Name=dyn_wing_front_l_1, swing1LimitDegrees=10, swing2LimitDegrees=20, twistLimitDegrees=5}
 {bone1Name=dyn_wing_back_r_2, bone2Name=dyn_wing_back_r_1, swing1LimitDegrees=15, swing2LimitDegrees=15, twistLimitDegrees=10}
 {bone1Name=dyn_wing_back_r_3, bone2Name=dyn_wing_back_r_1, swing1LimitDegrees=15, swing2LimitDegrees=15, twistLimitDegrees=10}
 {bone1Name=dyn_wing_back_r_4, bone2Name=dyn_wing_back_r_1, swing1LimitDegrees=15, swing2LimitDegrees=15, twistLimitDegrees=10}
 {bone1Name=dyn_wing_back_r_5, bone2Name=dyn_wing_back_r_1, swing1LimitDegrees=15, swing2LimitDegrees=15, twistLimitDegrees=10}
 {bone1Name=dyn_wing_back_r_6, bone2Name=dyn_wing_back_r_1, swing1LimitDegrees=15, swing2LimitDegrees=15, twistLimitDegrees=10}
 {bone1Name=dyn_wing_back_r_7, bone2Name=dyn_wing_back_r_1, swing1LimitDegrees=15, swing2LimitDegrees=15, twistLimitDegrees=10}
 {bone1Name=dyn_wing_front_r_2, bone2Name=dyn_wing_front_r_1, swing1LimitDegrees=15, swing2LimitDegrees=15, twistLimitDegrees=10}
 {bone1Name=dyn_wing_front_r_3, bone2Name=dyn_wing_front_r_1, swing1LimitDegrees=15, swing2LimitDegrees=15, twistLimitDegrees=10}
 {bone1Name=dyn_wing_front_r_4, bone2Name=dyn_wing_front_r_1, swing1LimitDegrees=15, swing2LimitDegrees=15, twistLimitDegrees=10}
 {bone1Name=dyn_wing_front_r_5, bone2Name=dyn_wing_front_r_1, swing1LimitDegrees=15, swing2LimitDegrees=15, twistLimitDegrees=10}
 {bone1Name=dyn_wing_front_r_6, bone2Name=dyn_wing_front_r_1, swing1LimitDegrees=15, swing2LimitDegrees=15, twistLimitDegrees=10}
 {bone1Name=dyn_wing_front_r_7, bone2Name=dyn_wing_front_r_1, swing1LimitDegrees=15, swing2LimitDegrees=15, twistLimitDegrees=10}
]
