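# Animation Blueprint for the male medium CanvasPrint cosmetic body, adding secondary motion to a pose copied from the attached parent mesh.
# The copied pose feeds a Control Rig, whose Dyn Fly input switches between 1 and 0 each update based on bIsSkyDiving.
# Weighted constraints then drive two clip bones from spine, pelvis and arm bones.
# AnimDynamics springs three grenade bones, then a RigidBody node simulates the body's physics asset relative to spine_05.
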AnimBlueprint /BRCosmetics/Characters/Player/Male/Medium/Bodies/M_MED_CanvasPrint/Meshes/M_MED_CanvasPrint_AnimBP
target_skeleton=&/M_MED_CanvasPrint_Skeleton
graphs [
 AnimGraph {nodes [
   n1 = Control Rig!(Source=Copy Pose From Mesh() {bUseAttachedParent=true}, Alpha, Dyn_fly:double=Get Dyn Fly() -> (Dyn Fly:double)) -> (Pose) {
    alphaBoolBlend {blendOption=Linear}
    controlRigAssetReference {blueprintRigClass=&/M_MED_CanvasPrint_CtrlRig.M_MED_CanvasPrint_CtrlRig_C}
   }
   n2 = "Constraint - dyn_lowClip (PPPPPPPP)"!(
    ComponentPose=Local To Component(LocalPose=@n1)
    Alpha=1
    ConstraintWeights_0=0.085
    ConstraintWeights_1:float=0.02
    ConstraintWeights_2:float=0.061
    ConstraintWeights_3:float=0.026
    ConstraintWeights_4:float=0.148
    ConstraintWeights_5:float=0.226
    ConstraintWeights_6:float=0.379
    ConstraintWeights_7:float=0.056
   ) -> (Pose) {
    alphaBoolBlend {blendOption=Linear}
    boneToModify {boneName=dyn_lowClip}
    constraintSetup [
     {targetBone {boneName=clavicle_l}}
     {targetBone {boneName=clavicle_r}}
     {targetBone {boneName=pelvis}}
     {targetBone {boneName=spine_01}}
     {targetBone {boneName=spine_02}}
     {targetBone {boneName=spine_03}}
     {targetBone {boneName=spine_04}}
     {targetBone {boneName=spine_05}}
    ]
    constraintWeights [1,1,1,1,1,1,1,1]
   }
   n3 = "Constraint - dyn_topClip (PPPPPPPP)"!(
    ComponentPose=@n2
    Alpha=1
    ConstraintWeights_0=0.239
    ConstraintWeights_1:float=0.025
    ConstraintWeights_2:float=0.038
    ConstraintWeights_3:float=0.023
    ConstraintWeights_4:float=0.099
    ConstraintWeights_5:float=0.194
    ConstraintWeights_6:float=0.296
    ConstraintWeights_7:float=0.085
   ) -> (Pose) {
    alphaBoolBlend {blendOption=Linear}
    boneToModify {boneName=dyn_topClip}
    constraintSetup [
     {targetBone {boneName=clavicle_l}}
     {targetBone {boneName=upperarm_twist_01_l}}
     {targetBone {boneName=pelvis}}
     {targetBone {boneName=spine_01}}
     {targetBone {boneName=spine_02}}
     {targetBone {boneName=spine_03}}
     {targetBone {boneName=spine_04}}
     {targetBone {boneName=spine_05}}
    ]
    constraintWeights [1,1,1,1,1,1,1,1]
   }
   n4 = Anim Dynamics - Bone: dyn_grenades_r(ComponentPose=@n3, Alpha=1, ExternalForce="0,0,0") {
    angularDampingOverride=1
    angularSpringConstant=950
    bAngularSpring=true
    bGravityOverrideInSimSpace=true
    bOverrideAngularDamping=true
    bOverrideLinearDamping=true
    simulationSpace=RootRelative
    alphaBoolBlend {blendOption=Linear}
    boundBone {boneName=dyn_grenades_r}
    componentAppliedLinearAccClamp=<400 400 400>
    componentLinearAccScale=<0.15 0.15 0.15>
    componentLinearVelScale=<0.15 0.15 0.15>
    physicsBodyDefinitions [{
      sphereCollisionRadius=0
      boundBone {boneName=dyn_grenades_r}
      boxExtents=<30 30 5>
      constraintSetup {angularTargetAxis=AxisZ, twistAxis=AxisZ, angularLimitsMax=<10 10 10>, angularLimitsMin=<-10 -10 -10>, angularTarget=<0 0 0>}
      localJointOffset=<-0.2 0 -5>
     }]
   }
   n5 = Anim Dynamics - Bone: dyn_grenades_1_l(ComponentPose=@n4, Alpha=1, ExternalForce="0,0,0") {
    angularDampingOverride=1
    angularSpringConstant=850
    bAngularSpring=true
    bGravityOverrideInSimSpace=true
    bOverrideAngularDamping=true
    bOverrideLinearDamping=true
    simulationSpace=RootRelative
    alphaBoolBlend {blendOption=Linear}
    boundBone {boneName=dyn_grenades_1_l}
    componentAppliedLinearAccClamp=<400 400 400>
    componentLinearAccScale=<0.15 0.15 0.15>
    componentLinearVelScale=<0.15 0.15 0.15>
    physicsBodyDefinitions [{
      sphereCollisionRadius=0
      boundBone {boneName=dyn_grenades_1_l}
      boxExtents=<40 40 8>
      constraintSetup {angularTargetAxis=AxisZ, twistAxis=AxisZ, angularLimitsMax=<10 10 10>, angularLimitsMin=<-10 -10 -10>, angularTarget=<0 0 0>}
      localJointOffset=<-0.2 0 -5>
     }]
   }
   n6 = Anim Dynamics - Bone: dyn_grenades_2_l(ComponentPose=@n5, Alpha=1, ExternalForce="0,0,0") {
    angularDampingOverride=1
    angularSpringConstant=900
    bAngularSpring=true
    bGravityOverrideInSimSpace=true
    bOverrideAngularDamping=true
    bOverrideLinearDamping=true
    simulationSpace=RootRelative
    alphaBoolBlend {blendOption=Linear}
    boundBone {boneName=dyn_grenades_2_l}
    componentAppliedLinearAccClamp=<400 400 400>
    componentLinearAccScale=<0.15 0.15 0.15>
    componentLinearVelScale=<0.15 0.15 0.15>
    physicsBodyDefinitions [{
      sphereCollisionRadius=0
      boundBone {boneName=dyn_grenades_2_l}
      boxExtents=<40 40 8>
      constraintSetup {angularTargetAxis=AxisZ, twistAxis=AxisZ, angularLimitsMax=<10 10 10>, angularLimitsMin=<-10 -10 -10>, angularTarget=<0 0 0>}
      localJointOffset=<-0.2 0 -5>
     }]
   }
   n7 = RigidBody(ComponentPose=@n6) {
    bOverrideWorldGravity=true
    overridePhysicsAsset=&/M_MED_CanvasPrint_Physics
    simulationSpace=BaseBoneSpace
    alphaBoolBlend {blendOption=Linear}
    alphaScaleBiasClamp {bInterpResult=true}
    baseBoneRef {boneName=spine_05}
    componentAppliedLinearAccClamp=<500 500 500>
    componentLinearAccScale=<1 1 1>
    componentLinearVelScale=<1 1 1>
    overrideWorldGravity {z=-120}
    simSpaceSettings {maxAngularAcceleration=100, maxAngularVelocity=10, maxLinearAcceleration=1000, maxLinearVelocity=100, velocityScaleZ=0.25, worldAlpha=0.5}
   }
   Output Pose(Result=Component To Local(ComponentPose=@n7)) {layerGroup=None}
  ]}
 EventGraph {nodes [
   Event BlueprintUpdateAnimation()
   > n1 = Branch(Condition=Get bIsSkyDiving())
   > Set Dyn Fly(Dyn Fly:double=1) -> (Output_Get:double)
   Set Dyn Fly(execute=@n1.else, Dyn Fly=0)
  ]}
]
variables [Dyn Fly {type=number}]
