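# Animation Blueprint for the face accessory part of the male medium CanvasPrint character body.
# It copies the pose from the attached parent mesh, then applies a constraint that ties the dyn_head_correct bone to the head bone.
# Each animation update sets the constraint's alpha to full strength while an emote plays and to a partial blend otherwise.
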AnimBlueprint /BRCosmetics/Characters/Player/Male/Medium/Bodies/M_MED_CanvasPrint/Meshes/Parts/M_MED_CanvasPrint_FaceAcc_AnimBP
target_skeleton=&/M_MED_CanvasPrint_FaceAcc_Skeleton
graphs [
 AnimGraph {nodes [
   n1 = "Constraint - dyn_head_correct (P)"(
    ComponentPose=Local To Component(LocalPose=Copy Pose From Mesh() {bUseAttachedParent=true})
    Alpha=Get Dyn_Consrt_Head_Alpha() -> (Dyn_Consrt_Head_Alpha:float)
    ConstraintWeights_0=1
   ) {
    alphaBoolBlend {blendOption=Linear}
    boneToModify {boneName=dyn_head_correct}
    constraintSetup [{targetBone {boneName=head}}]
    constraintWeights [1]
   }
   Output Pose(Result=Component To Local(ComponentPose=@n1)) {layerGroup=None}
  ]}
 EventGraph {nodes [
   Event BlueprintUpdateAnimation()
   > n1 = Branch(Condition=Get bIsPlayingEmote())
   > Set Dyn_Consrt_Head_Alpha(Dyn_Consrt_Head_Alpha:float=1) -> (Output_Get:float)
   Set Dyn_Consrt_Head_Alpha(execute=@n1.else, Dyn_Consrt_Head_Alpha=0.7)
  ]}
]
variables [Dyn_Consrt_Head_Alpha {type=number, default=0.7}]
